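# Niagara idle effect for the FlintTwirl mic stand instrument cosmetic, adding blue glows, sparkles and rotating rings.
# BaseGlows_MicGlow layers partitioned sprite, mesh and light glows at base and mic, gating one burst on User.IsInstrument and scaling alpha by time of day and front-end preview.
# StaticRings_LG spins two mesh rings at the base with orbiting glow sprites.
# Sparkle motes spawn around base, pole and mic, with GPU and lower-rate CPU variants.
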NiagaraSystem /SparksCosmetics/Instrument/MicStand/FlintTwirl/Effects/NS_MicStand_FlintTwirl_Idle_FX
dependencies {
 usedDataInterfaces [
  &NiagaraDataInterfaceCurve
  &NiagaraDataInterfaceArrayFloat
  &NiagaraDataInterfaceArrayFloat3
  &NiagaraDataInterfaceColorCurve
 ]
 usedDynamicInputs [&RandomRangeInt, &ColorFromCurve]
 usedModules [
  &SystemState
  &self:SystemSpawnScript.SystemScriptSource.SystemScriptGraph.NiagaraNodeAssignment_1.SetVariables_#1
  &self:SystemSpawnScript.SystemScriptSource.SystemScriptGraph.NiagaraNodeAssignment_2.SetVariables_#2
 ]
}
emitters {
 BaseGlows_MicGlow {
  emitterUpdateScript [
   &EmitterState()
   &SpawnBurst_Instantaneous(
    Spawn Count=4
    Spawn Group=1
    Spawn Probability=&MakeCustomFloatFromBool(Bool=&SetBoolByFloatComparison(A=$User.IsInstrument, B=0), True Float=0, False Float=1)
    Use Spawn Probability=true
    UseLoopCountLimit=true
   ) {module=SpawnBurst_Instantaneous001}
   &SpawnBurst_Instantaneous(Spawn Count=5, Use Spawn Probability=false, UseLoopCountLimit=true)
  ]
  particleSpawnScript [
   &PartitionParticles(Number Of Partitions=5)
   /SetVariables(Particles.VisibilityTag=$Particles.PartitionParticles.Partition) {isSetParameters=true}
   &InitializeParticle(
    Color=&MergeHelper_ColorSwitch(
     0=&MergeHelper_ColorSwitch(
      0=<rgba 0 0.1892 1 0.72>
      1=<rgba 0 1.5016 5 0.5>
      2=<rgba 0.0135 2.9653 4 0.5>
      3=<rgba 0 1.853 5 1>
      Index=$Particles.PartitionParticles.Partition
      Num=::NewEnumerator2
     )
     1=&MergeHelper_ColorSwitch(
      0=<rgba 0 0.1892 1 0.72>
      1=<rgba 0 1.5016 5 0.5>
      2=<rgba 0.0135 2.9653 4 0.2>
      3=<rgba 0 1.853 5 1>
      Index=$Particles.PartitionParticles.Partition
      Num=::NewEnumerator2
     )
     2=?
     3=?
     Index=&SpawnGroup()
     Num=::NewEnumerator0
    )
    Position Offset=&MaskVectorBySpawnGroup(Masked Vector=<0 0 15>, Passthrough Vector=<0 -11.5 139.5>)
    Sprite Size Mode=::NewEnumerator3
    Uniform Sprite Size=&MergeHelper_FloatSwitch(
     0=&MergeHelper_FloatSwitch(0=27, 1=10, 2=18, 3=24, Index=$Particles.PartitionParticles.Partition, Num=::NewEnumerator2)
     1=&MergeHelper_FloatSwitch(0=20, 1=10.5, 2=19, 3=18, Index=$Particles.PartitionParticles.Partition, Num=::NewEnumerator2)
     2=?
     3=?
     Index=&SpawnGroup()
     Num=::NewEnumerator0
    )
    UsePositionOffset=true
    Write Material Random=true
   )
   &CameraOffset(Camera Offset Amount=&MergeHelper_FloatSwitch(
     0=&MergeHelper_FloatSwitch(0=0, 1=-5, 2=-10, 3=10, Index=$Particles.PartitionParticles.Partition, Num=::NewEnumerator2)
     1=&MergeHelper_FloatSwitch(0=0, 1=5, 2=-10, 3=10, Index=$Particles.PartitionParticles.Partition, Num=::NewEnumerator2)
     2=?
     3=?
     Index=&SpawnGroup()
     Num=::NewEnumerator0
    ))
   &FN_Sprite_Attributes(
    Camera Offset=15
    Color=&MakeLinearColorFromVectorAndFloat(
     "Vector (RGB)"=&MakeVectorFromLinearColorRGB(Color=$Particles.Color)
     "Float (Alpha)"=&MaskFloatBySpawnGroup(Masked Float=0.25, Passthrough Float=0.12)
    )
    Create Camera Offset Attribute=true
    Create Color Attribute=true
    Create SpriteSize Attribute=true
    SpriteSize=&Multiply_Vector2D_ByFloat(Float=2.25, Vector2D=$Particles.SpriteSize)
    Write to Module Namespace=true
   )
   /SetVariables(Particles.HaloDynamic:Vector4f=&MakeVector4(X=1, Y=0, Z=1, W=0)) {isSetParameters=true}
  ]
  particleUpdateScript [
   &ParticleState(Kill Particles When Lifetime Has Elapsed=false)
   &ScaleSpriteSize(Scale Sprite Size Mode=::NewEnumerator0, Uniform Scale Factor=$System.EngineOwnerScale)
   &SolveForcesAndVelocity(Clamp Velocity=false, Limit Acceleration=false)
   /SetVariables(Particles.RayDynamicParameter:Vector4f=&MakeVector4(X=$Emitter.Age, Y=0.1, Z=0, W=0)) {isSetParameters=true}
   &ScaleColor(
    Scale Alpha=&Multiply_Float(
     A=&MergeHelper_FloatSwitch(
      0=&FloatFromCurve(
       FloatCurve=DataInterface() {propertyValues {Curve {keys [
           {interpMode=RCIM_Cubic, leaveTangentWeight=0.1001, tangentMode=RCTM_User, time=0, value=0.9}
           {arriveTangentWeight=0.1021, interpMode=RCIM_Cubic, leaveTangentWeight=0.0679, tangentMode=RCTM_User, time=0.3003, value=1}
           {arriveTangentWeight=0.0666, interpMode=RCIM_Cubic, leaveTangentWeight=0.0667, tangentMode=RCTM_User, time=0.5, value=1}
           {arriveTangent=-0.2, interpMode=RCIM_Cubic, leaveTangent=-0.2, time=0.7, value=0.95}
           {arriveTangentWeight=0.1, interpMode=RCIM_Cubic, tangentMode=RCTM_User, time=1, value=0.9}
          ]}}}
       CurveIndex=Linked() {linkedVariable {name=NPC.FortniteNPC.RemappedTimeOfDay_0-1, type=NiagaraFloat}}
       Scale Curve=&MakeCustomFloatFromBool(Bool=$User.bIsFrontEndPreview, True Float=0.7, False Float=1)
      )
      1=&FloatFromCurve(
       FloatCurve=DataInterface() {propertyValues {Curve {
          defaultValue=1
          keys [
           {interpMode=RCIM_Cubic, tangentMode=RCTM_User, time=0, value=0.9}
           {arriveTangentWeight=0.1021, interpMode=RCIM_Cubic, leaveTangentWeight=0.0679, tangentMode=RCTM_User, time=0.3003, value=1}
           {arriveTangentWeight=0.1667, interpMode=RCIM_Cubic, leaveTangentWeight=0.1667, tangentMode=RCTM_User, time=0.5, value=1}
           {arriveTangent=-0.2, interpMode=RCIM_Cubic, leaveTangent=-0.2, time=0.7, value=0.95}
           {arriveTangentWeight=0.3333, interpMode=RCIM_Cubic, tangentMode=RCTM_User, time=1, value=0.9}
          ]
         }}}
       CurveIndex=Linked() {linkedVariable {name=NPC.FortniteNPC.RemappedTimeOfDay_0-1, type=NiagaraFloat}}
       Scale Curve=&MakeCustomFloatFromBool(Bool=$User.bIsFrontEndPreview, True Float=0.7, False Float=1)
      )
      2=&FloatFromCurve(
       FloatCurve=DataInterface() {propertyValues {Curve {
          defaultValue=1
          keys [
           {interpMode=RCIM_Cubic, tangentMode=RCTM_User, time=0, value=0.85}
           {
            arriveTangent=0.3983
            arriveTangentWeight=0.1001
            interpMode=RCIM_Cubic
            leaveTangent=0.3983
            leaveTangentWeight=0.0666
            tangentMode=RCTM_User
            time=0.2225
            value=0.9562
           }
           {arriveTangentWeight=0.1667, interpMode=RCIM_Cubic, leaveTangentWeight=0.1667, tangentMode=RCTM_User, time=0.5, value=1}
           {arriveTangent=-0.5195, interpMode=RCIM_Cubic, leaveTangent=-0.5195, tangentMode=RCTM_User, time=0.7977, value=0.92}
           {arriveTangentWeight=0.3333, interpMode=RCIM_Cubic, tangentMode=RCTM_User, time=1, value=0.85}
          ]
         }}}
       CurveIndex=Linked() {linkedVariable {name=NPC.FortniteNPC.RemappedTimeOfDay_0-1, type=NiagaraFloat}}
       Scale Curve=&MakeCustomFloatFromBool(Bool=$User.bIsFrontEndPreview, True Float=0.5, False Float=1)
      )
      3=?
      Index=$Particles.PartitionParticles.Partition
      Num=::NewEnumerator1
     )
     B=&MaskFloatBySpawnGroup(Masked Float=1, Passthrough Float=1)
    )
    ScaleA=true
    ScaleRGB=true
   )
   /SetVariables(
    Particles.FN_Sprite_Attributes.SpriteSize:Vector2f=&Multiply_Vector2D_ByFloat(
     Float=&MakeFloatFromVector(Vector=$Engine.Owner.Scale, Channel=::NewEnumerator0)
     Vector2D=&Multiply_Vector2D_ByFloat(
      Float=&Waveform("Waveform [1]"=::NewEnumerator0, "[1] Phase Scale"=2.5, "[1] Frequency"=1, "[1] Amplitude Min/Max"=<0.98 1>)
      Vector2D=$Particles.Initial.FN_Sprite_Attributes.SpriteSize
     )
    )
    Particles.HaloColor:LinearColor=&Multiply_LinearColorByFloat(
     Linear Color=$Particles.FN_Sprite_Attributes.Color
     Float=&Waveform("Waveform [1]"=::NewEnumerator12, "[1] Frequency"=5, "[1] Amplitude Min/Max"=<0.85 1>)
    )
   ) {isSetParameters=true}
   &FN_Light_Attributes(
    Color=<rgba 0 1.853 5 1>
    Create Color Attribute=true
    Create Exponent Attribute=true
    Create Position Attribute=true
    Create Radius Attribute=true
    Exponent=3
    Position=<0 0 15.5>
    Radius=65
   )
  ]
  rendererClasses [
   &NiagaraSpriteRendererProperties
   &NiagaraMeshRendererProperties
   &NiagaraLightRendererProperties
  ]
  renderers [
   {rendererClass=&NiagaraSpriteRendererProperties}
   {rendererClass=&NiagaraSpriteRendererProperties, rendererIndex=1}
   {rendererClass=&NiagaraSpriteRendererProperties, rendererIndex=2}
   {rendererClass=&NiagaraMeshRendererProperties, rendererIndex=3}
   {rendererClass=&NiagaraSpriteRendererProperties, rendererIndex=4}
   {rendererClass=&NiagaraLightRendererProperties, rendererIndex=5}
  ]
 }
 Motes_Sparkles {
  simTarget=GPUComputeSim
  emitterUpdateScript [
   &EmitterState()
   &SpawnRate(SpawnGroup=1, SpawnRate=12, Use Spawn Probability=false) {module=SpawnRate001}
   &SpawnRate(Spawn Probability=0.33, SpawnRate=15, Use Spawn Probability=true)
  ]
  particleSpawnScript [
   /SetVariables(Particles.VisibilityTag=&SpawnGroup()) {isSetParameters=true}
   &PartitionParticles(Number Of Partitions=3)
   &ShapeLocation(
    Apply To Particle Position=::NewEnumerator1
    MaskSpawnGroup=false
    Override Seed=false
    Shape Origin=&AddVectorToPosition(Position=<0 0 0>, Vector=<0 0 15>)
    Sphere Radius=&MergeHelper_FloatSwitch(
     0=&RandomRangeFloat(Minimum=12, Maximum=15)
     1=&RandomRangeFloat(Minimum=7, Maximum=12)
     2=?
     3=?
     Index=$Emitter.SpawnGroup
     Num=::NewEnumerator0
    )
    Sphere Surface Distribution=1
    UseOwnerScale=true
   )
   &ShapeLocation(
    Cylinder Height=11
    Cylinder Radius=11
    MaskSpawnGroup=false
    Override Seed=false
    Rotation Mode=::NewEnumerator1
    Shape Origin=&AddVectorToPosition(Position=<0 0 0>, Vector=<0 -2 139>)
    Shape Primitive=::NewEnumerator1
    Surface Only=true
    UseOwnerScale=true
   ) {module=ShapeLocation001}
   &ShapeLocation(
    Apply To Particle Position=::NewEnumerator1
    Cylinder Height=75
    Cylinder Radius=7
    MaskSpawnGroup=false
    Override Seed=false
    Shape Origin=&AddVectorToPosition(Position=<0 0 0>, Vector=<0 0 70>)
    Shape Primitive=::NewEnumerator1
    Surface Only=true
    Use Endcaps In Surface Only Mode=false
    UseOwnerScale=true
   ) {module=ShapeLocation002}
   &InitializeParticle(
    Color=&MergeHelper_ColorSwitch(0=<rgba 0.3423 1.5331 3 1>, 1=<rgba 0.5705 2.5552 5 1>, 2=?, 3=?, Index=&SpawnGroup(), Num=::NewEnumerator0)
    Lifetime=&MergeHelper_FloatSwitch(
     0=&RandomRangeFloat(Minimum=0.7, Maximum=1)
     1=&RandomRangeFloat(Minimum=1, Maximum=2)
     2=?
     3=?
     Index=&SpawnGroup()
     Num=::NewEnumerator0
    )
    Position Offset=&MaskVectorBySpawnGroup(
     Masked Vector=&MergeHelper_VectorSwitch(
      0=Linked() {linkedVariable {name=Particles.ShapeLocation.ShapePosition, type=NiagaraPosition}}
      1=Linked() {linkedVariable {name=Particles.ShapeLocation002.ShapePosition, type=NiagaraPosition}}
      2=Linked() {linkedVariable {name=Particles.ShapeLocation001.ShapePosition, type=NiagaraPosition}}
      3=Linked() {linkedVariable {name=Particles.ShapeLocation002.ShapePosition, type=NiagaraPosition}}
      Index=$Particles.PartitionParticles.Partition
      Num=::NewEnumerator1
     )
     Passthrough Vector=Linked() {linkedVariable {name=Particles.ShapeLocation.ShapePosition, type=NiagaraPosition}}
    )
    Sprite Size Mode=::NewEnumerator3
    Uniform Sprite Size=&MergeHelper_FloatSwitch(
     0=&RandomRangeFloat(Minimum=8, Maximum=15)
     1=&RandomRangeFloat(Minimum=1, Maximum=2)
     2=?
     3=?
     Index=$Emitter.SpawnGroup
     Num=::NewEnumerator0
    )
    UsePositionOffset=true
    Write Material Random=true
   )
   &CameraOffset(Camera Offset Amount=&MergeHelper_FloatSwitch(
     0=&RandomRangeFloat(Minimum=15, Maximum=20)
     1=&RandomRangeFloat(Minimum=2, Maximum=5)
     2=?
     3=?
     Index=$Emitter.SpawnGroup
     Num=::NewEnumerator0
    ))
   &DynamicMaterialParameters(
    Index 0 Param 1=0.25
    Index 0 Param 2=&MergeHelper_FloatSwitch(
     0=0.3
     1=&RandomRangeFloat(Minimum=0.3, Maximum=0.5)
     2=?
     3=?
     Index=$Emitter.SpawnGroup
     Num=::NewEnumerator0
    )
    Index 0 Param 3=&RandomRangeFloat(Minimum=0.25, Maximum=0.5)
    Index 0 Param 4=1
    Module.Index0Float0Write=true
    Module.Index0Float1Write=true
    Module.Index0Float2Write=true
    Module.Index0Float3Write=true
   )
  ]
  particleUpdateScript [
   &ParticleState()
   &ScaleSpriteSize(
    Uniform Curve Scale=$System.EngineOwnerScale
    Uniform Curve Sprite Scale=DataInterface() {propertyValues {Curve {
       defaultValue=0
       keys [
        {arriveTangent=11.4403, interpMode=RCIM_Cubic, leaveTangent=11.4403, tangentMode=RCTM_User, time=0, value=0}
        {arriveTangent=0.6136, interpMode=RCIM_Cubic, leaveTangent=0.6136, tangentMode=RCTM_User, time=0.114, value=0.968}
        {arriveTangent=-1.0926, interpMode=RCIM_Cubic, leaveTangent=-1.0926, time=0.5233, value=0.8079}
        {arriveTangent=0.0224, interpMode=RCIM_Cubic, leaveTangent=0.0224, tangentMode=RCTM_User, time=1, value=0}
       ]
      }}}
   )
   &ScaleColor(
    Scale Alpha=&FloatFromCurve(FloatCurve=DataInterface() {propertyValues {Curve {keys [
         <at 0 0 RCIM_Cubic>
         {arriveTangent=1.9996, interpMode=RCIM_Cubic, leaveTangent=1.9996, time=0.0968, value=1}
         {arriveTangent=-1.1072, interpMode=RCIM_Cubic, leaveTangent=-1.1072, time=0.4966, value=0.9931}
         <at 1 0 RCIM_Cubic>
        ]}}})
    ScaleA=true
    ScaleRGB=true
   )
  ]
  rendererClasses [&NiagaraSpriteRendererProperties]
  renderers [
   {rendererClass=&NiagaraSpriteRendererProperties}
   {rendererClass=&NiagaraSpriteRendererProperties, rendererIndex=1}
  ]
 }
 Motes_Sparkles_Low {
  emitterUpdateScript [
   &EmitterState()
   &SpawnRate(SpawnGroup=1, SpawnRate=4, Use Spawn Probability=false) {module=SpawnRate001}
   &SpawnRate(Spawn Probability=0.33, SpawnRate=15, Use Spawn Probability=true)
  ]
  particleSpawnScript [
   /SetVariables(Particles.VisibilityTag=&SpawnGroup()) {isSetParameters=true}
   &PartitionParticles(Number Of Partitions=3)
   &ShapeLocation(
    Apply To Particle Position=::NewEnumerator1
    MaskSpawnGroup=false
    Override Seed=false
    Shape Origin=&AddVectorToPosition(Position=<0 0 0>, Vector=<0 0 15>)
    Sphere Radius=&MergeHelper_FloatSwitch(
     0=&RandomRangeFloat(Minimum=12, Maximum=15)
     1=&RandomRangeFloat(Minimum=7, Maximum=12)
     2=?
     3=?
     Index=$Emitter.SpawnGroup
     Num=::NewEnumerator0
    )
    Sphere Surface Distribution=1
    UseOwnerScale=true
   )
   &ShapeLocation(
    Cylinder Height=11
    Cylinder Radius=11
    MaskSpawnGroup=false
    Override Seed=false
    Rotation Mode=::NewEnumerator1
    Shape Origin=&AddVectorToPosition(Position=<0 0 0>, Vector=<0 -2 139>)
    Shape Primitive=::NewEnumerator1
    Surface Only=true
    UseOwnerScale=true
   ) {module=ShapeLocation001}
   &ShapeLocation(
    Apply To Particle Position=::NewEnumerator1
    Cylinder Height=75
    Cylinder Radius=7
    MaskSpawnGroup=false
    Override Seed=false
    Shape Origin=&AddVectorToPosition(Position=<0 0 0>, Vector=<0 0 70>)
    Shape Primitive=::NewEnumerator1
    Surface Only=true
    Use Endcaps In Surface Only Mode=false
    UseOwnerScale=true
   ) {module=ShapeLocation002}
   &InitializeParticle(
    Color=&MergeHelper_ColorSwitch(0=<rgba 0.3423 1.5331 3 1>, 1=<rgba 0.5705 2.5552 5 1>, 2=?, 3=?, Index=&SpawnGroup(), Num=::NewEnumerator0)
    Lifetime=&MergeHelper_FloatSwitch(
     0=&RandomRangeFloat(Minimum=0.7, Maximum=1)
     1=&RandomRangeFloat(Minimum=1, Maximum=2)
     2=?
     3=?
     Index=&SpawnGroup()
     Num=::NewEnumerator0
    )
    Position Offset=&MaskVectorBySpawnGroup(
     Masked Vector=&MergeHelper_VectorSwitch(
      0=Linked() {linkedVariable {name=Particles.ShapeLocation.ShapePosition, type=NiagaraPosition}}
      1=Linked() {linkedVariable {name=Particles.ShapeLocation002.ShapePosition, type=NiagaraPosition}}
      2=Linked() {linkedVariable {name=Particles.ShapeLocation001.ShapePosition, type=NiagaraPosition}}
      3=Linked() {linkedVariable {name=Particles.ShapeLocation002.ShapePosition, type=NiagaraPosition}}
      Index=$Particles.PartitionParticles.Partition
      Num=::NewEnumerator1
     )
     Passthrough Vector=Linked() {linkedVariable {name=Particles.ShapeLocation.ShapePosition, type=NiagaraPosition}}
    )
    Sprite Size Mode=::NewEnumerator3
    Uniform Sprite Size=&MergeHelper_FloatSwitch(
     0=&RandomRangeFloat(Minimum=8, Maximum=15)
     1=&RandomRangeFloat(Minimum=1, Maximum=2)
     2=?
     3=?
     Index=$Emitter.SpawnGroup
     Num=::NewEnumerator0
    )
    UsePositionOffset=true
    Write Material Random=true
   )
   &CameraOffset(Camera Offset Amount=&MergeHelper_FloatSwitch(
     0=&RandomRangeFloat(Minimum=15, Maximum=20)
     1=&RandomRangeFloat(Minimum=2, Maximum=5)
     2=?
     3=?
     Index=$Emitter.SpawnGroup
     Num=::NewEnumerator0
    ))
   &DynamicMaterialParameters(
    Index 0 Param 1=0.25
    Index 0 Param 2=&MergeHelper_FloatSwitch(
     0=0.3
     1=&RandomRangeFloat(Minimum=0.3, Maximum=0.5)
     2=?
     3=?
     Index=$Emitter.SpawnGroup
     Num=::NewEnumerator0
    )
    Index 0 Param 3=&RandomRangeFloat(Minimum=0.25, Maximum=0.5)
    Index 0 Param 4=1
    Module.Index0Float0Write=true
    Module.Index0Float1Write=true
    Module.Index0Float2Write=true
    Module.Index0Float3Write=true
   )
  ]
  particleUpdateScript [
   &ParticleState()
   &ScaleSpriteSize(
    Uniform Curve Scale=$System.EngineOwnerScale
    Uniform Curve Sprite Scale=DataInterface() {propertyValues {Curve {
       defaultValue=0
       keys [
        {arriveTangent=11.4403, interpMode=RCIM_Cubic, leaveTangent=11.4403, tangentMode=RCTM_User, time=0, value=0}
        {arriveTangent=0.6136, interpMode=RCIM_Cubic, leaveTangent=0.6136, tangentMode=RCTM_User, time=0.114, value=0.968}
        {arriveTangent=-1.0926, interpMode=RCIM_Cubic, leaveTangent=-1.0926, time=0.5233, value=0.8079}
        {arriveTangent=0.0224, interpMode=RCIM_Cubic, leaveTangent=0.0224, tangentMode=RCTM_User, time=1, value=0}
       ]
      }}}
   )
   &ScaleColor(
    Scale Alpha=&FloatFromCurve(FloatCurve=DataInterface() {propertyValues {Curve {keys [
         <at 0 0 RCIM_Cubic>
         {arriveTangent=1.9996, interpMode=RCIM_Cubic, leaveTangent=1.9996, time=0.0968, value=1}
         {arriveTangent=-1.1072, interpMode=RCIM_Cubic, leaveTangent=-1.1072, time=0.4966, value=0.9931}
         <at 1 0 RCIM_Cubic>
        ]}}})
    ScaleA=true
    ScaleRGB=true
   )
  ]
  rendererClasses [&NiagaraSpriteRendererProperties]
  renderers [
   {rendererClass=&NiagaraSpriteRendererProperties}
   {rendererClass=&NiagaraSpriteRendererProperties, rendererIndex=1}
  ]
 }
 StaticRings_LG {
  emitterUpdateScript [
   &EmitterState()
   &SpawnBurst_Instantaneous(Spawn Count=2, Use Spawn Probability=false, UseLoopCountLimit=true)
  ]
  particleSpawnScript [
   &PartitionParticles()
   &InitializeParticle(
    Color=<rgba 0.0213 4.8741 11 1>
    Mesh Scale Mode=::NewEnumerator3
    Mesh Uniform Scale=&Multiply_Float(
     A=&Multiply_Float(
      A=&SelectFloatFromArray(Direct Array Index=$Particles.PartitionParticles.Partition, Float Selection Array=DataInterface() {propertyValues {FloatData [0.21,0.23]}})
      B=0.65
     )
     B=1
    )
    Position Offset=<0 0 15>
    Position Offset Coordinate Space=::Local
    Sprite Size Mode=::NewEnumerator3
    Uniform Sprite Size=&Multiply_Float(A=14, B=$System.EngineOwnerScale)
    UsePositionOffset=true
    Write Material Random=true
   )
   &InitialMeshOrientation(
    Mesh Axis To Orient=&SelectVectorFromArray(
     Direct Array Index=$Particles.PartitionParticles.Partition
     Vector Selection Array=DataInterface() {propertyValues {FloatData [<1 1 0>, <-1 -1 0>]}}
    )
    Mesh Orientation Mode=::NewEnumerator2
    Orientation Vector=&SelectVectorFromArray(
     Direct Array Index=$Particles.PartitionParticles.Partition
     Vector Selection Array=DataInterface() {propertyValues {FloatData [<1 0 0>, <0 0 0>]}}
    )
    Rotation=&SelectVectorFromArray(
     Direct Array Index=$Particles.PartitionParticles.Partition
     Vector Selection Array=DataInterface() {propertyValues {FloatData [<-0.045 -0.0405 0>, <-0.045 -0.0405 0>]}}
    )
    Use Rotation Vector=true
   )
   /SetVariables(Particles.RingGlowOffset:NiagaraPosition=&TransformPosition(Position=<0 0 15>, Source Space=::Local, Destination Space=::Simulation)) {isSetParameters=true}
   &DynamicMaterialParameters(
    Index 0 Param 1=0.25
    Index 0 Param 2=&MergeHelper_FloatSwitch(
     0=0.3
     1=&RandomRangeFloat(Minimum=0.3, Maximum=0.5)
     2=?
     3=?
     Index=$Emitter.SpawnGroup
     Num=::NewEnumerator0
    )
    Index 0 Param 3=&RandomRangeFloat(Minimum=0.25, Maximum=0.5)
    Index 0 Param 4=1
    Module.Index0Float0Write=true
    Module.Index0Float1Write=true
    Module.Index0Float2Write=true
    Module.Index0Float3Write=true
   )
  ]
  particleUpdateScript [
   &SolveForcesAndVelocity(Clamp Velocity=false, Limit Acceleration=false)
   &ParticleState(Kill Particles When Lifetime Has Elapsed=false, Let Infinitely Lived Particles Die When Emitter Deactivates=true)
   &UpdateMeshOrientation(Rotation Rate=&Lerp_Float(A=0.25, B=0.65, Alpha=0), Rotation Vector=<0 0 1>)
   &ScaleMeshSize()
   &ScaleColor(ScaleA=true, ScaleRGB=true) {enabled=false}
   /SetVariables(Particles.SoftGlowColor:LinearColor=&Multiply_LinearColorByFloat(Linear Color=$Particles.Color, Float=0.5)) {isSetParameters=true}
   /SetVariables(
    Particles.RingGlowColor:LinearColor=<rgba 0 1.6296 3 1>
    Particles.SpritePos:NiagaraPosition=&TransformPosition(
     Position=&AddVectorToPosition(
      Position=&RotateVectorByQuaternion(
       Vector To Rotate=&SelectVectorFromArray(
        Direct Array Index=$Particles.PartitionParticles.Partition
        Vector Selection Array=DataInterface() {propertyValues {FloatData [<10 10 0>, <11 11 0>]}}
       )
       Quaternion=$Output.UpdateMeshOrientation.MeshOrientation
      )
      Vector=&ConvertVectorToPosition(
       Input Position=Linked() {linkedVariable {name=Particles.RingGlowOffset, type=NiagaraPosition}}
       Conversion Operation:ENiagara_LWCConvertVecToPos=::NewEnumerator0
      )
     )
     Source Space=::Local
     Destination Space=::Simulation
    )
   ) {isSetParameters=true}
  ]
  rendererClasses [&NiagaraMeshRendererProperties, &NiagaraSpriteRendererProperties]
  renderers [
   {rendererClass=&NiagaraMeshRendererProperties, rendererIndex=0}
   {rendererClass=&NiagaraMeshRendererProperties, rendererIndex=1}
   {rendererClass=&NiagaraSpriteRendererProperties, rendererIndex=2}
   {rendererClass=&NiagaraMeshRendererProperties, rendererIndex=3}
  ]
 }
}
summary {
 systemName=NS_MicStand_FlintTwirl_Idle_FX
 emitters [
  {
   emitterName=BaseGlows_MicGlow
   rendererClasses [
    &NiagaraSpriteRendererProperties
    &NiagaraMeshRendererProperties
    &NiagaraLightRendererProperties
   ]
  }
  {emitterName=StaticRings_LG, rendererClasses [&NiagaraMeshRendererProperties, &NiagaraSpriteRendererProperties]}
  {emitterName=Motes_Sparkles, simTarget=GPUComputeSim, rendererClasses [&NiagaraSpriteRendererProperties]}
  {emitterName=Motes_Sparkles_Low, rendererClasses [&NiagaraSpriteRendererProperties]}
 ]
 userVariables [
  User.bIsFrontEndPreview {type=NiagaraBool}
  User.IsInstrument {type=NiagaraFloat, defaultValue=0, description=""}
 ]
}
