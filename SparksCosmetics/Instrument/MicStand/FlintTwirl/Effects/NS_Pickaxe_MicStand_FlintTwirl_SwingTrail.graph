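# Swing trail effect for the FlintTwirl MicStand pickaxe cosmetic, combining ribbon trails, a constellation and sparkles.
# Two AnimTrail ribbon emitters are tinted by a system time-of-day color, with alpha scaled by time of day or a front-end preview flag.
# Constellation sprites with a ribbon, plus curl-noise sparkles, spawn at positions sampled through a socket reader.
# Sparkles come in a distance-spawned GPU emitter and a constant-rate LowMedium variant, apparently for quality scaling.
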NiagaraSystem /SparksCosmetics/Instrument/MicStand/FlintTwirl/Effects/NS_Pickaxe_MicStand_FlintTwirl_SwingTrail
dependencies {
 usedDataInterfaces [&NiagaraDataInterfaceCurve, &NiagaraDataInterfaceColorCurve, &NiagaraDataInterfaceSocketReader]
 usedDynamicInputs []
 usedModules [
  &self:SystemSpawnScript.SystemScriptSource.SystemScriptGraph.NiagaraNodeAssignment_0.SetVariables_#1
  &self:SystemSpawnScript.SystemScriptSource.SystemScriptGraph.NiagaraNodeAssignment_1.SetVariables_#2
  &SystemState
  &self:SystemSpawnScript.SystemScriptSource.SystemScriptGraph.NiagaraNodeAssignment_2.SetVariables_#3
  &self:SystemSpawnScript.SystemScriptSource.SystemScriptGraph.NiagaraNodeAssignment_3.SetVariables_#4
 ]
}
emitters {
 Constellation {
  emitterUpdateScript [
   &EmitterState(Life Cycle Mode=::NewEnumerator1, Loop Behavior=::NewEnumerator1, Loop Duration=1, UseLoopDelay=false)
   &SpawnPerUnit(
    Max Movement Threshold=4000
    Module.Use Max Movement Threshold=false
    Module.Use Movement Tolerance=false
    Spawn Probability=1
    Spawn Spacing=30
    Use Spawn Probability=false
   )
  ]
  particleSpawnScript [
   &InitializeParticle(
    Color=<rgba 0 0.3286 1 1>
    Lifetime=0.35
    Ribbon Width=0.75
    Ribbon Width Mode=::NewEnumerator1
    Sprite Size Mode=::NewEnumerator3
    Uniform Sprite Size=&RandomRangeFloat(Minimum=15, Maximum=20)
    UsePositionOffset=false
    Write Material Random=true
   )
   &SocketLocation(
    Interpolate Socket Sampling=true
    SamplePosOffset=true
    Sampled Position Offset=<20 0 0>
    Socket Reader=$System.Socket Reader
    Socket Sampling Mode=::NewEnumerator2
    VelocitySpeedLimit=false
   )
   &JitterPosition(Jitter Amount=12)
   &CameraOffset(Camera Offset Amount=7)
   &DynamicMaterialParameters(
    Index 0 Param 1=0.25
    Index 0 Param 2=&RandomRangeFloat(Minimum=0.1, Maximum=0.2)
    Index 0 Param 3=0.25
    Module.Index0Float0Write=true
    Module.Index0Float1Write=true
    Module.Index0Float2Write=true
    Module.Index0Float3Write=true
   )
  ]
  particleUpdateScript [
   &ParticleState()
   &ScaleSpriteSize(Uniform Curve Sprite Scale=DataInterface() {propertyValues {Curve {keys [{arriveTangent=-0.3913, interpMode=RCIM_Cubic, leaveTangent=-0.3913, tangentMode=RCTM_User, time=0, value=0}, <at 0.3 1 RCIM_Cubic>]}}})
   &ScaleColor(
    Scale Alpha=&FloatFromCurve(
     FloatCurve=DataInterface() {propertyValues {Curve {
        defaultValue=1
        keys [
         {arriveTangent=9.0725, interpMode=RCIM_Cubic, leaveTangent=9.0725, tangentMode=RCTM_User, time=0, value=0}
         {arriveTangent=-2.5249, interpMode=RCIM_Cubic, leaveTangent=-2.5249, tangentMode=RCTM_User, time=0.1408, value=0.9847}
         {arriveTangent=-0.0405, interpMode=RCIM_Cubic, leaveTangent=-0.0405, tangentMode=RCTM_User, time=0.5, value=0}
        ]
       }}}
     CurveIndex=$System.NormalizedLoopAge
    )
    Scale RGB=&MakeVectorFromLinearColorRGB(Color=&ColorFromCurve(ColorCurve=DataInterface() {propertyValues {
        AlphaCurve {keys [<at 0 1 RCIM_Cubic>,<at 0.2876 0.7995>,<at 0.9134 0 RCIM_Cubic>]}
        BlueCurve {keys [<at 0.162 22 RCIM_Cubic>,<at 0.4693 1>,<at 1 1 RCIM_Cubic>]}
        GreenCurve {keys [<at 0.162 22 RCIM_Cubic>,<at 0.4693 1>,<at 1 1 RCIM_Cubic>]}
        RedCurve {keys [<at 0.162 22 RCIM_Cubic>,<at 0.4693 1>,<at 1 1 RCIM_Cubic>]}
       }}))
    ScaleA=true
    ScaleRGB=true
   )
   /SetVariables(Particles.ConstColor:LinearColor=&Multiply_LinearColorByFloat(Linear Color=$Particles.Color, Float=0.25)) {isSetParameters=true}
  ]
  rendererClasses [&NiagaraSpriteRendererProperties, &NiagaraRibbonRendererProperties]
  renderers [
   {rendererClass=&NiagaraSpriteRendererProperties}
   {rendererClass=&NiagaraRibbonRendererProperties, rendererIndex=1}
  ]
 }
 Sparkles_GPU {
  simTarget=GPUComputeSim
  emitterUpdateScript [
   &EmitterState()
   &SpawnPerUnit(
    Module.Use Max Movement Threshold=true
    Module.Use Movement Tolerance=false
    Spawn Probability=1
    Spawn Spacing:NiagaraFloat=&RandomRangeFloat(Minimum=9, Maximum=13)
    Use Spawn Probability=false
   )
  ]
  particleSpawnScript [
   &InitializeParticle(
    Color=<rgba 0.2349 1.4154 2 1>
    Lifetime=&FloatFromCurve(
     FloatCurve=DataInterface() {propertyValues {Curve {keys [<at 0.2 0.15>,<at 0.6 0.2>,<at 0.8 0.2735>,<at 1.0018 0.6>]}}}
     CurveIndex=&RandomRangeFloat(Minimum=0, Maximum=1)
    )
    Sprite Rotation Mode=::NewEnumerator2
    Sprite Size Mode=::NewEnumerator3
    Uniform Sprite Size=&FloatFromCurve(
     FloatCurve=DataInterface() {propertyValues {Curve {keys [<at 0 7>,<at 0.8 10>,<at 0.9991 15>]}}}
     CurveIndex=&RandomRangeFloat(Minimum=0, Maximum=1)
    )
    UsePositionOffset=false
    Write Material Random=true
   )
   &SocketLocation(
    Inherited Velocity Speed Limit=11
    Interpolate Socket Sampling=true
    SamplePosOffset=true
    Sampled Orientation Coordinate Space:ENiagaraCoordinateSpace=::World
    Sampled Position Offset=<0 0 10>
    Sampled Velocity Coordinate Space:ENiagaraCoordinateSpace=::Local
    Socket Reader=$System.Socket Reader
    Socket Sampling Mode=::NewEnumerator2
    VelocitySpeedLimit=false
   )
   &ShapeLocation(MaskSpawnGroup=false, Override Seed=false, Shape Origin=$Output.SocketLocation.SampledPosition, Sphere Radius=13, UseOwnerScale=true)
   &DynamicMaterialParameters(
    Index 0 Param 1=0.2
    Index 0 Param 2=&RandomRangeFloat(Minimum=0.3, Maximum=0.5)
    Index 0 Param 3=0.25
    Module.Index0Float0Write=true
    Module.Index0Float1Write=true
    Module.Index0Float2Write=true
    Module.Index0Float3Write=true
   )
  ]
  particleUpdateScript [
   &ParticleState()
   &ScaleSpriteSize(Uniform Curve Sprite Scale=DataInterface() {propertyValues {Curve {
       defaultValue=0
       keys [
        {arriveTangent=4.1785, interpMode=RCIM_Cubic, leaveTangent=4.1785, tangentMode=RCTM_User, time=0, value=0}
        {arriveTangentWeight=0.0834, interpMode=RCIM_Cubic, leaveTangentWeight=0.1167, tangentMode=RCTM_User, time=0.25, value=1}
        {arriveTangentWeight=0.1944, interpMode=RCIM_Cubic, leaveTangentWeight=0.2222, tangentMode=RCTM_User, time=0.5, value=0.9975}
        <at 1 0 RCIM_Cubic>
       ]
      }}})
   &ScaleColor(
    Scale Alpha=&FloatFromCurve(
     FloatCurve=DataInterface() {propertyValues {Curve {keys [
         {arriveTangent=12.5818, interpMode=RCIM_Cubic, leaveTangent=12.5818, tangentMode=RCTM_User, time=0, value=0}
         {arriveTangentWeight=0.0883, interpMode=RCIM_Cubic, leaveTangentWeight=0.3532, tangentMode=RCTM_User, time=0.2, value=0.883}
         <at 1 0 RCIM_Cubic>
        ]}}}
     Scale Curve=&RandomRangeFloat(Minimum=0.5, Maximum=1)
    )
    ScaleA=true
    ScaleRGB=true
   )
   &CurlNoiseForce(Bias Noise=false, Noise Strength=111, Pan Noise=false)
   &Drag(Drag=&RandomRangeFloat(Minimum=2, Maximum=3), Use Linear Drag=true, Use Rotational Drag=true)
   &SolveForcesAndVelocity(Clamp Velocity=false, Limit Acceleration=false)
  ]
  rendererClasses [&NiagaraSpriteRendererProperties]
  renderers [{rendererClass=&NiagaraSpriteRendererProperties}]
 }
 Sparkles_LowMedium {
  emitterUpdateScript [
   &EmitterState()
   &SpawnRate(SpawnRate=30, Use Spawn Probability=false)
  ]
  particleSpawnScript [
   &InitializeParticle(
    Color=<rgba 0.2349 1.4154 2 1>
    Lifetime=&FloatFromCurve(
     FloatCurve=DataInterface() {propertyValues {Curve {keys [<at 0.2 0.15>,<at 0.6 0.2>,<at 0.8 0.2735>,<at 1.0018 0.6>]}}}
     CurveIndex=&RandomRangeFloat(Minimum=0, Maximum=1)
    )
    Sprite Rotation Mode=::NewEnumerator2
    Sprite Size Mode=::NewEnumerator3
    Uniform Sprite Size=&FloatFromCurve(
     FloatCurve=DataInterface() {propertyValues {Curve {keys [<at 0 7>,<at 0.8 10>,<at 0.9991 15>]}}}
     CurveIndex=&RandomRangeFloat(Minimum=0, Maximum=1)
    )
    UsePositionOffset=false
    Write Material Random=true
   )
   &SocketLocation(
    Inherited Velocity Speed Limit=11
    Interpolate Socket Sampling=true
    SamplePosOffset=true
    Sampled Orientation Coordinate Space:ENiagaraCoordinateSpace=::World
    Sampled Position Offset=<0 0 10>
    Sampled Velocity Coordinate Space:ENiagaraCoordinateSpace=::Local
    Socket Reader=$System.Socket Reader
    Socket Sampling Mode=::NewEnumerator2
    VelocitySpeedLimit=false
   )
   &ShapeLocation(MaskSpawnGroup=false, Override Seed=false, Shape Origin=$Output.SocketLocation.SampledPosition, Sphere Radius=13, UseOwnerScale=true)
   &DynamicMaterialParameters(
    Index 0 Param 1=0.2
    Index 0 Param 2=&RandomRangeFloat(Minimum=0.3, Maximum=0.5)
    Index 0 Param 3=0.25
    Module.Index0Float0Write=true
    Module.Index0Float1Write=true
    Module.Index0Float2Write=true
    Module.Index0Float3Write=true
   )
  ]
  particleUpdateScript [
   &ParticleState()
   &ScaleSpriteSize(Uniform Curve Sprite Scale=DataInterface() {propertyValues {Curve {
       defaultValue=0
       keys [
        {arriveTangent=4.1785, interpMode=RCIM_Cubic, leaveTangent=4.1785, tangentMode=RCTM_User, time=0, value=0}
        {arriveTangentWeight=0.0834, interpMode=RCIM_Cubic, leaveTangentWeight=0.1167, tangentMode=RCTM_User, time=0.25, value=1}
        {arriveTangentWeight=0.1944, interpMode=RCIM_Cubic, leaveTangentWeight=0.2222, tangentMode=RCTM_User, time=0.5, value=0.9975}
        <at 1 0 RCIM_Cubic>
       ]
      }}})
   &ScaleColor(
    Scale Alpha=&FloatFromCurve(
     FloatCurve=DataInterface() {propertyValues {Curve {keys [
         {arriveTangent=12.5818, interpMode=RCIM_Cubic, leaveTangent=12.5818, tangentMode=RCTM_User, time=0, value=0}
         {arriveTangentWeight=0.0883, interpMode=RCIM_Cubic, leaveTangentWeight=0.3532, tangentMode=RCTM_User, time=0.2, value=0.883}
         <at 1 0 RCIM_Cubic>
        ]}}}
     Scale Curve=&RandomRangeFloat(Minimum=0.5, Maximum=1)
    )
    ScaleA=true
    ScaleRGB=true
   )
   &CurlNoiseForce(Bias Noise=false, Noise Strength=111, Pan Noise=false)
   &Drag(Drag=&RandomRangeFloat(Minimum=2, Maximum=3), Use Linear Drag=true, Use Rotational Drag=true)
   &SolveForcesAndVelocity(Clamp Velocity=false, Limit Acceleration=false)
  ]
  rendererClasses [&NiagaraSpriteRendererProperties]
  renderers [{rendererClass=&NiagaraSpriteRendererProperties}]
 }
 Trail_Color {
  emitterUpdateScript [
   &EmitterState()
   &FN_SpawnAnimTrail_v2()
  ]
  particleSpawnScript [
   /Game/Effects/Niagara/Modules/AnimTrail/FN_InitializeAnimTrail/FN_InitializeAnimTrailBoneIndecies(First Bone or Socket Index=1)
   /Game/Effects/Niagara/Modules/AnimTrail/FN_InitializeAnimTrail/FN_InitializeAnimTrail(AnimTrailWidth=1, Color=<rgba 1 1 1 0.5>, Lifetime=0.25)
   /SetVariables(Particles.GalaxyPosition:NiagaraPosition=&AddVectorToPosition(Position=$Particles.Position, Vector=<0 0 0>)) {isSetParameters=true}
   &AddVelocity(Velocity=<0 500 0>) {enabled=false}
  ]
  particleUpdateScript [
   &ScaleColor(
    Scale Alpha=&FloatFromCurve(
     FloatCurve=DataInterface() {propertyValues {Curve {keys [
         {arriveTangent=8.4839, interpMode=RCIM_Cubic, leaveTangent=8.4839, tangentMode=RCTM_User, time=0.0042, value=0.1618}
         {
          arriveTangentWeight=0.0359
          interpMode=RCIM_Cubic
          leaveTangent=-1.8179
          leaveTangentWeight=0.1016
          tangentMode=RCTM_Break
          time=0.15
          value=1
         }
         {arriveTangent=-0.4123, interpMode=RCIM_Cubic, leaveTangent=-0.4123, tangentMode=RCTM_User, time=0.35, value=0.15}
         {arriveTangentWeight=0.2167, interpMode=RCIM_Cubic, tangentMode=RCTM_Break, time=1, value=0}
        ]}}}
     Scale Curve=&MakeCustomFloatFromBool(
      Bool=$User.bIsFrontEndPreview
      True Float=0.4
      False Float=&FloatFromCurve(
       FloatCurve=DataInterface() {propertyValues {Curve {keys [
           <at 0 0.7 RCIM_Cubic>
           {arriveTangent=0.6011, interpMode=RCIM_Cubic, leaveTangent=0.6011, time=0.2, value=0.9}
           <at 0.4991 1 RCIM_Cubic>
           {arriveTangent=-0.5989, interpMode=RCIM_Cubic, leaveTangent=-0.5989, time=0.6969, value=0.9}
           <at 1 0.7 RCIM_Cubic>
          ]}}}
       CurveIndex=Linked() {linkedVariable {name=NPC.FortniteNPC.TimeOfDay_0-1, type=NiagaraFloat}}
      )
     )
    )
    Scale RGB=&MakeVectorFromLinearColorRGB(Color=$System.TOD Color)
    ScaleA=true
    ScaleRGB=true
   ) {module=ScaleColor002}
   &FN_UpdateAnimTrail_v2()
   /SetVariables(Particles.GalaxyColor:LinearColor=&MakeLinearColorFromVectorAndFloat(
     "Vector (RGB)"=<1 1 1>
     "Float (Alpha)"=&FloatFromCurve(
      FloatCurve=DataInterface() {propertyValues {Curve {
         defaultValue=1
         keys [{interpMode=RCIM_Cubic, tangentMode=RCTM_User, time=0, value=1}, {interpMode=RCIM_Cubic, tangentMode=RCTM_User, time=1, value=0}]
        }}}
      Scale Curve=&MakeCustomFloatFromBool(Bool=$User.bIsFrontEndPreview, True Float=0.12, False Float=0.25)
     )
    )) {isSetParameters=true}
   &DynamicMaterialParameters(
    Index 0 Param 1=1
    Index 1 Param 3=&FloatFromCurve(
     FloatCurve=DataInterface() {propertyValues {Curve {keys [<at 0 0>, {arriveTangent=1.9814, tangentMode=RCTM_User, time=1, value=1}]}}}
     CurveIndex=&ModuloFloat(A=$Emitter.Age, B=1)
     Scale Curve=0
    )
    Index 1 Param 4=1
    Module.Index0Float0Write=true
    Module.Index0Float1Write=true
    Module.Index0Float2Write=true
    Module.Index0Float3Write=true
    Module.Index1Float0Write=true
    Module.Index1Float1Write=true
    Module.Index1Float2Write=true
    Module.Index1Float3Write=true
    Write Parameter Index 1=::NewEnumerator1
   )
   &SolveForcesAndVelocity(Clamp Velocity=false, Limit Acceleration=false) {enabled=false}
  ]
  rendererClasses [&NiagaraRibbonRendererProperties]
  renderers [
   {rendererClass=&NiagaraRibbonRendererProperties}
   {rendererClass=&NiagaraRibbonRendererProperties, rendererIndex=1}
  ]
 }
 Trail_Color001 {
  emitterUpdateScript [
   &EmitterState()
   &FN_SpawnAnimTrail_v2()
  ]
  particleSpawnScript [
   &FN_InitializeAnimTrailBoneIndecies(First Bone or Socket Index=1)
   &FN_InitializeAnimTrail(AnimTrailWidth=1, Color=<rgba 1 1 1 0.7>, Lifetime=0.2)
   &AddVelocity(Velocity=<0 500 0>) {enabled=false}
  ]
  particleUpdateScript [
   &FN_UpdateAnimTrail_v2()
   &ScaleColor(
    Scale Alpha=&FloatFromCurve(
     FloatCurve=DataInterface() {propertyValues {Curve {keys [
         {interpMode=RCIM_Cubic, leaveTangent=-4.1145, tangentMode=RCTM_Break, time=0, value=1}
         {arriveTangent=-0.4123, interpMode=RCIM_Cubic, leaveTangent=-0.4123, tangentMode=RCTM_User, time=0.35, value=0.15}
         {arriveTangent=-0.2139, interpMode=RCIM_Cubic, tangentMode=RCTM_Break, time=1, value=0}
        ]}}}
     Scale Curve=&MakeCustomFloatFromBool(
      Bool=$User.bIsFrontEndPreview
      True Float=0.6
      False Float=&FloatFromCurve(
       FloatCurve=DataInterface() {propertyValues {Curve {keys [
           <at 0 0.7 RCIM_Cubic>
           {arriveTangent=0.6011, interpMode=RCIM_Cubic, leaveTangent=0.6011, time=0.2, value=0.9}
           <at 0.4991 1 RCIM_Cubic>
           {arriveTangent=-0.5989, interpMode=RCIM_Cubic, leaveTangent=-0.5989, time=0.6969, value=0.9}
           <at 1 0.7 RCIM_Cubic>
          ]}}}
       CurveIndex=Linked() {linkedVariable {name=NPC.FortniteNPC.TimeOfDay_0-1, type=NiagaraFloat}}
      )
     )
    )
    Scale RGB=&MakeVectorFromLinearColorRGB(Color=$System.TOD Color)
    ScaleA=true
    ScaleRGB=true
   ) {module=ScaleColor002}
   &DynamicMaterialParameters(
    Index 0 Param 4=1
    Index 1 Param 3=&FloatFromCurve(
     FloatCurve=DataInterface() {propertyValues {Curve {keys [<at 0 0>, {arriveTangent=1.9814, tangentMode=RCTM_User, time=1, value=1}]}}}
     CurveIndex=&ModuloFloat(A=$Emitter.Age, B=1)
     Scale Curve=0
    )
    Index 1 Param 4=1
    Module.Index0Float0Write=true
    Module.Index0Float1Write=true
    Module.Index0Float2Write=true
    Module.Index0Float3Write=true
    Module.Index1Float0Write=true
    Module.Index1Float1Write=true
    Module.Index1Float2Write=true
    Module.Index1Float3Write=true
    Write Parameter Index 1=::NewEnumerator1
   )
   &SolveForcesAndVelocity(Clamp Velocity=false, Limit Acceleration=false) {enabled=false}
  ]
  rendererClasses [&NiagaraRibbonRendererProperties]
  renderers [{rendererClass=&NiagaraRibbonRendererProperties}]
 }
}
summary {
 systemName=NS_Pickaxe_MicStand_FlintTwirl_SwingTrail
 emitters [
  {emitterName=Trail_Color, rendererClasses [&NiagaraRibbonRendererProperties]}
  {emitterName=Trail_Color001, rendererClasses [&NiagaraRibbonRendererProperties]}
  {
   emitterName=Constellation
   rendererClasses [&NiagaraSpriteRendererProperties, &NiagaraRibbonRendererProperties]
  }
  {emitterName=Sparkles_GPU, simTarget=GPUComputeSim, rendererClasses [&NiagaraSpriteRendererProperties]}
  {emitterName=Sparkles_LowMedium, rendererClasses [&NiagaraSpriteRendererProperties]}
 ]
 userVariables [
  User.bIsFrontEndPreview {type=NiagaraBool}
  User.RibbonIDOffset {type=NiagaraInt32}
  User.SkeletalMesh {
   type=NiagaraDataInterfaceSkeletalMesh
   defaultValue {
    struct=NiagaraExt_VariableValue_DataInterface
    value {dataInterface=&self:NiagaraDataInterfaceSkeletalMesh_0}
   }
  }
 ]
}
