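# Physics asset for the FlintTwirl mic stand cosmetic, defining simulation bodies and joints for its decorative bows and ribbons.
# A long handle capsule anchors small capsules on left and right bows, each with upper and lower pieces, plus left and right ribbon bases.
# Each ribbon base branches into two-link front and behind chains.
# Narrow swing limits keep motion subtle, with twist locked everywhere except the first front-ribbon links, which allow slight twist.
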PhysicsAsset /SparksCosmetics/Instrument/MicStand/FlintTwirl/Meshes/FlintTwirl_MicStand_Physics
bodies {
 dyn_Bow_L [{shapeType=Capsule, length=2, radius=1, shapeName=dyn_Bow_L_capsule, center=<0.1395 -0.4829 0.1374>, rotation=<rot 19.1346 -90 -25.1638>}]
 dyn_Bow_L_Lower [{shapeType=Capsule, length=2, radius=1, shapeName=dyn_Bow_L_Lower_capsule, center=<1.5003 -0.1055 -0.1832>, rotation=<rot -3.2048 -90 -2.695>}]
 dyn_Bow_L_Upper [{shapeType=Capsule, length=2, radius=1, shapeName=dyn_Bow_L_Upper_capsule, center=<1.1178 0.0866 -0.0952>, rotation=<rot -3.2423 -90 -5.1633>}]
 dyn_Bow_R [{shapeType=Capsule, length=2, radius=1, shapeName=dyn_Bow_R_capsule, center=<0.0137 0.5852 0.1084>, rotation=<rot -16.8383 -90 -4.4661>}]
 dyn_Bow_R_Lower [{shapeType=Capsule, length=2, radius=1, shapeName=dyn_Bow_R_Lower_capsule, center=<0.9401 0.1477 -0.0962>, rotation=<rot 1.8317 -90 -1.0231>}]
 dyn_Bow_R_Upper [{shapeType=Capsule, length=2, radius=1, shapeName=dyn_Bow_R_Upper_capsule, center=<1.1092 -0.0148 -0.1212>, rotation=<rot 4.4238 -90 -0.8354>}]
 dyn_Ribbon_Behind_L_01 [{shapeType=Capsule, length=2, radius=1, shapeName=dyn_Ribbon_Behind_L_01_capsule, rotation=<rot 90 0 0>}]
 dyn_Ribbon_Behind_L_02 [{shapeType=Capsule, length=3, radius=1, shapeName=dyn_Ribbon_Behind_L_02_capsule, center=<0.55 0 0>, rotation=<rot 90 0 0>}]
 dyn_Ribbon_Behind_R_01 [{shapeType=Capsule, length=2, radius=1, shapeName=dyn_Ribbon_Behind_R_01_capsule, rotation=<rot 90 0 0>}]
 dyn_Ribbon_Behind_R_02 [{shapeType=Capsule, length=3, radius=1, shapeName=dyn_Ribbon_Behind_R_02_capsule, center=<0.55 0 0>, rotation=<rot 90 0 0>}]
 dyn_Ribbon_Front_L_01 [{shapeType=Capsule, length=2, radius=1, shapeName=dyn_Ribbon_Front_L_01_capsule, center=<0.6793 0 0>, rotation=<rot 90 0 0>}]
 dyn_Ribbon_Front_L_02 [{shapeType=Capsule, length=2.7439, radius=1, shapeName=dyn_Ribbon_Front_L_02_capsule, center=<0.5571 -0.4573 0>, rotation=<rot 90 0 0>}]
 dyn_Ribbon_Front_R_01 [{shapeType=Capsule, length=2, radius=1, shapeName=dyn_Ribbon_Front_R_01_capsule, rotation=<rot 90 0 0>}]
 dyn_Ribbon_Front_R_02 [{shapeType=Capsule, length=3, radius=1, shapeName=dyn_Ribbon_Front_R_02_capsule, center=<0.5469 0 0>, rotation=<rot 90 0 0>}]
 dyn_Ribbon_L_Base [{shapeType=Capsule, length=1, radius=1, shapeName=dyn_Ribbon_L_Base_capsule, rotation=<rot 90 0 0>}]
 dyn_Ribbon_R_Base [{shapeType=Capsule, length=1, radius=1, shapeName=dyn_Ribbon_R_Base_capsule, rotation=<rot 90 0 0>}]
 handle [{shapeType=Capsule, length=46.09, radius=1.21, center=<0 0 28.7391>}]
}
constraints [
 {bone1Name=dyn_Bow_L, bone2Name=handle, swing1LimitDegrees=6, swing2LimitDegrees=8, twistLimitDegrees=45, twistMotion=Locked}
 {bone1Name=dyn_Bow_L_Lower, bone2Name=dyn_Bow_L, swing1LimitDegrees=4, swing2LimitDegrees=8, twistLimitDegrees=45, twistMotion=Locked}
 {bone1Name=dyn_Bow_L_Upper, bone2Name=dyn_Bow_L, swing1LimitDegrees=12, swing2LimitDegrees=8, twistLimitDegrees=45, twistMotion=Locked}
 {bone1Name=dyn_Bow_R, bone2Name=handle, swing1LimitDegrees=6, swing2LimitDegrees=8, twistLimitDegrees=45, twistMotion=Locked}
 {bone1Name=dyn_Bow_R_Lower, bone2Name=dyn_Bow_R, swing1LimitDegrees=4, swing2LimitDegrees=8, twistLimitDegrees=45, twistMotion=Locked}
 {bone1Name=dyn_Bow_R_Upper, bone2Name=dyn_Bow_R, swing1LimitDegrees=12, swing2LimitDegrees=8, twistLimitDegrees=45, twistMotion=Locked}
 {bone1Name=dyn_Ribbon_L_Base, bone2Name=handle, swing1LimitDegrees=10, swing2LimitDegrees=10, twistLimitDegrees=45, twistMotion=Locked}
 {
  bone1Name=dyn_Ribbon_Behind_L_01
  bone2Name=dyn_Ribbon_L_Base
  swing1LimitDegrees=15
  swing2LimitDegrees=15
  twistLimitDegrees=45
  twistMotion=Locked
 }
 {
  bone1Name=dyn_Ribbon_Behind_L_02
  bone2Name=dyn_Ribbon_Behind_L_01
  swing1LimitDegrees=10
  swing2LimitDegrees=20
  twistLimitDegrees=45
  twistMotion=Locked
 }
 {bone1Name=dyn_Ribbon_Front_L_01, bone2Name=dyn_Ribbon_L_Base, swing1LimitDegrees=20, swing2LimitDegrees=25, twistLimitDegrees=5}
 {
  bone1Name=dyn_Ribbon_Front_L_02
  bone2Name=dyn_Ribbon_Front_L_01
  swing1LimitDegrees=10
  swing2LimitDegrees=20
  twistLimitDegrees=45
  twistMotion=Locked
 }
 {bone1Name=dyn_Ribbon_R_Base, bone2Name=handle, swing1LimitDegrees=10, swing2LimitDegrees=10, twistLimitDegrees=45, twistMotion=Locked}
 {
  bone1Name=dyn_Ribbon_Behind_R_01
  bone2Name=dyn_Ribbon_R_Base
  swing1LimitDegrees=15
  swing2LimitDegrees=15
  twistLimitDegrees=45
  twistMotion=Locked
 }
 {
  bone1Name=dyn_Ribbon_Behind_R_02
  bone2Name=dyn_Ribbon_Behind_R_01
  swing1LimitDegrees=10
  swing2LimitDegrees=20
  twistLimitDegrees=45
  twistMotion=Locked
 }
 {bone1Name=dyn_Ribbon_Front_R_01, bone2Name=dyn_Ribbon_R_Base, swing1LimitDegrees=20, swing2LimitDegrees=25, twistLimitDegrees=5}
 {
  bone1Name=dyn_Ribbon_Front_R_02
  bone2Name=dyn_Ribbon_Front_R_01
  swing1LimitDegrees=15
  swing2LimitDegrees=20
  twistLimitDegrees=45
  twistMotion=Locked
 }
]
