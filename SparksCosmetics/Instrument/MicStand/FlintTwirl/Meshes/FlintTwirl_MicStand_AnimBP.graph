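# Animation Blueprint for the FlintTwirl mic stand cosmetic, adding procedural ornament motion and secondary physics.
# Each update derives sine-wave offsets for four star bones and continuous rotation for two crescents and the stand star from real time.
# Chained additive bone-space Modify Bone nodes apply these, then feed a RigidBody node simulating relative to the handle bone.
# A skydiving flag swaps that node's gravity override from straight down to a sideways, upward vector.
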AnimBlueprint /SparksCosmetics/Instrument/MicStand/FlintTwirl/Meshes/FlintTwirl_MicStand_AnimBP
target_skeleton=&/FlintTwirl_MicStand_Skeleton
graphs [
 AnimGraph {nodes [
   n1 = "Transform (Modify) Bone - Bone: dyn_Star_01"(
    ComponentPose=Local To Component(LocalPose=Input Pose())
    Alpha=1
    Translation=Get Star Float 01() -> (Star Float 01:Vector)
    Rotation="0,0,0"
    Scale="1,1,1"
   ) {translationMode=BMM_Additive, translationSpace=BCS_BoneSpace, alphaBoolBlend {blendOption=Linear}, boneToModify {boneName=dyn_Star_01}}
   n2 = "Transform (Modify) Bone - Bone: dyn_Star_02"(ComponentPose=@n1, Alpha=1, Translation=Get Star Float 02() -> (Star Float 02:Vector), Rotation="0,0,0", Scale="1,1,1") {translationMode=BMM_Additive, translationSpace=BCS_BoneSpace, alphaBoolBlend {blendOption=Linear}, boneToModify {boneName=dyn_Star_02}}
   n3 = "Transform (Modify) Bone - Bone: dyn_Star_03"(ComponentPose=@n2, Alpha=1, Translation=Get Star Float 03() -> (Star Float 03:Vector), Rotation="0,0,0", Scale="1,1,1") {translationMode=BMM_Additive, translationSpace=BCS_BoneSpace, alphaBoolBlend {blendOption=Linear}, boneToModify {boneName=dyn_Star_03}}
   n4 = "Transform (Modify) Bone - Bone: dyn_Stand_Star"(
    ComponentPose=@n3
    Alpha=1
    Translation=Get Star Float 04() -> (Star Float 04:Vector)
    Rotation=Get Star Base Rotator() -> (Star Base Rotator:Rotator)
    Scale="1,1,1"
   ) {
    rotationMode=BMM_Additive
    rotationSpace=BCS_BoneSpace
    translationMode=BMM_Additive
    translationSpace=BCS_BoneSpace
    alphaBoolBlend {blendOption=Linear}
    boneToModify {boneName=dyn_Stand_Star}
   }
   n5 = "Transform (Modify) Bone - Bone: dyn_Crescent_inner"(ComponentPose=@n4, Alpha=1, Translation="0,0,0", Rotation=Get Crescent Rotator inner() -> (Crescent Rotator inner:Rotator), Scale="1,1,1") {rotationMode=BMM_Additive, rotationSpace=BCS_BoneSpace, alphaBoolBlend {blendOption=Linear}, boneToModify {boneName=dyn_Crescent_inner}}
   n6 = "Transform (Modify) Bone - Bone: dyn_Crescent_outer"(ComponentPose=@n5, Alpha=1, Translation="0,0,0", Rotation=Get Crescent Rotator outer() -> (Crescent Rotator outer:Rotator), Scale="1,1,1") {rotationMode=BMM_Additive, rotationSpace=BCS_BoneSpace, alphaBoolBlend {blendOption=Linear}, boneToModify {boneName=dyn_Crescent_outer}}
   n7 = RigidBody!(ComponentPose=@n6, Alpha, OverrideWorldGravity:Vector=Get Gravity Override(), ExternalForce) -> (Pose) {
    bClampLinearTranslationLimitToRefPose=true
    bOverrideWorldGravity=true
    lODThreshold=1
    overridePhysicsAsset=&/FlintTwirl_MicStand_Physics
    simulationSpace=BaseBoneSpace
    alphaBoolBlend {blendOption=Linear}
    baseBoneRef {boneName=handle}
    componentAppliedLinearAccClamp=<3000 3000 3000>
    componentLinearAccScale=<0.5 0.5 0.5>
    componentLinearVelScale=<0.3 0.3 0.3>
    overrideWorldGravity {z=-980}
   }
   Output Pose(Result=Component To Local(ComponentPose=@n7))
  ]}
 EventGraph {nodes [
   Event BlueprintUpdateAnimation()
   > n1 = Sequence()
   > n2 = Branch(Condition=Get bIsSkyDiving())
   > Set Gravity Override(Gravity Override="0,-300,200")
   Set Gravity Override(execute=@n2.else, Gravity Override="0,0,-980")
   n3 = GetRealTimeSeconds()
   n4 = "Sin (Radians)"(A=float * float(A=float * float(A=0.5, B=0.5), B=@n3))
   n5 = "Sin (Radians)"(A=float * float(A=float * float(A=1, B=1), B=@n3))
   Set Star Float 01(execute=@n1.then_1, Star Float 01:Vector=MakeVector(Y=float * float(A=@n5, B=0.7), Z=float * float(A=@n4, B=0.7))) -> (Output_Get:Vector)
   n6 = "Sin (Radians)"(A=float * float(A=float * float(A=0.5, B=0.5), B=@n3))
   n7 = "Sin (Radians)"(A=float * float(A=float * float(A=1.5, B=1.5), B=@n3))
   > Set Star Float 02(Star Float 02:Vector=MakeVector(Y=float * float(A=@n7, B=0.7), Z=float * float(A=@n6, B=0.7))) -> (Output_Get:Vector)
   n8 = "Sin (Radians)"(A=float * float(A=float * float(A=0.5, B=0.5), B=@n3))
   n9 = "Sin (Radians)"(A=float * float(A=float * float(A=1, B=1), B=@n3))
   > Set Star Float 03(Star Float 03:Vector=MakeVector(Y=float * float(A=@n9, B=0.75), Z=float * float(A=@n8, B=0.7))) -> (Output_Get:Vector)
   n10 = "Sin (Radians)"(A=float * float(A=float * float(A=0.7, B=2) {is=title K2Node_PromotableOperator}, B=@n3))
   > Set Star Float 04(Star Float 04:Vector=MakeVector(Y=0, Z=float * float(A=@n10, B=0.55))) -> (Output_Get:Vector)
   > Set Crescent Rotator inner(Crescent Rotator inner:Rotator=MakeRotator(Pitch=float * float(A=@n3, B=50) {is=title K2Node_CommutativeAssociativeBinaryOperator})) -> (Output_Get:Rotator)
   > Set Crescent Rotator outer(Crescent Rotator outer:Rotator=MakeRotator(Pitch=float * float(A=@n3, B=50))) -> (Output_Get:Rotator)
   n11 = float / float(A=@n3, B=-5) {is=title K2Node_CallFunction}
   > Set Star Base Rotator(Star Base Rotator:Rotator=MakeRotator(Yaw=float * float(A=@n11, B=100))) -> (Output_Get:Rotator)
   TryGetPawnOwner()
  ]}
]
variables [
 Crescent Rotator inner {type=Rotator}
 Crescent Rotator outer {type=Rotator}
 Gravity Override {type=Vector}
 Star Base Rotator {type=Rotator}
 Star Float 01 {type=Vector}
 Star Float 02 {type=Vector}
 Star Float 03 {type=Vector}
 Star Float 04 {type=Vector}
]
